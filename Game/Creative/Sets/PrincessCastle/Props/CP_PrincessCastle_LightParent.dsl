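; Parent Blueprint for Princess Castle light props in Creative, built on a time-of-day aware base.
; It overrides the parent's time-of-day default variables, fixing volumetric light scattering intensity and keeping time-of-day lights and material emissive values enabled.
; The shadow-related settings pass through from the parent unchanged.
; The construction script destroys the GlowFlicker component unless UseGlowFlickerCard is set.
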(fn GetTimeOfDayBlueprintDefaultVariables
  (bind _outvariables (|CallParentFunction))
  (bind (_outvariables_alternateshadowstaticmesh _outvariables_volumetriclightscatteringintensity _outvariables_bdisabletodlightsandmaterialemissivevalues _outvariables_bdisablestaticmeshshadowcastingwhenlightsareactive _outvariables_buseanalternateshadowmeshwhenthelightisoff _outvariables_bcastvolumetricshadows) (Utilities|Struct|BreakTimeOfDayBlueprintDefaultVariables _outvariables))
  (return (Utilities|Struct|MakeTimeOfDayBlueprintDefaultVariables _outvariables_alternateshadowstaticmesh 1.0 false _outvariables_bdisablestaticmeshshadowcastingwhenlightsareactive _outvariables_buseanalternateshadowmeshwhenthelightisoff _outvariables_bcastvolumetricshadows)))

(fn ConstructionScript
  (if (not (Variables|Default|GetUseGlowFlickerCard))
    (Components|DestroyComponent (Variables|Default|GetGlowFlicker))))
